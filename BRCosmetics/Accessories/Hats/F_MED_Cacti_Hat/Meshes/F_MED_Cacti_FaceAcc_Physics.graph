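# Physics asset for the cactus hat face accessory F_MED_Cacti, simulating its two branches on a character's head.
# Capsule bodies cover the head and the left and right dynamic branch bones.
# Each branch is constrained directly to the head with narrow swing limits and almost no twist, allowing slight wobble.
# The left branch has a larger capsule and slightly looser swing than the right.
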PhysicsAsset /BRCosmetics/Accessories/Hats/F_MED_Cacti_Hat/Meshes/F_MED_Cacti_FaceAcc_Physics
bodies {
 dyn_left_cactusbranch_02 [{shapeType=Capsule, length=7.927, radius=3.6331, center=<-1.364 0.8262 7.1806>, rotation=<rot 7.4297 -90 -15.7931>}]
 dyn_right_cactusbranch_02 [{shapeType=Capsule, length=5.9146, radius=2.6351, center=<0.5963 0.2605 5.539>, rotation=<rot 6.3672 -90 10.3887>}]
 head [{shapeType=Capsule, length=8.8136, radius=8.9427, center=<7.3091 0.1896 0.0512>, rotation=<rot -10.1615 -89.9998 -87.7511>}]
}
constraints [
 {bone1Name=dyn_right_cactusbranch_02, bone2Name=head, swing1LimitDegrees=10, swing2LimitDegrees=10, twistLimitDegrees=1}
 {bone1Name=dyn_left_cactusbranch_02, bone2Name=head, swing1LimitDegrees=15, swing2LimitDegrees=15, twistLimitDegrees=1}
]
