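# Animation Blueprint for the F_MED_Cacti face accessory in the hat cosmetics, adding secondary physics motion to the attached mesh.
# It copies the pose from the attached parent mesh, then runs a RigidBody node with its own physics asset, overridden gravity and translation clamped to the reference pose.
# Each update sets Hat_Alpha from bIsTargeting, reducing the simulation's blend weight while targeting and restoring full weight otherwise.
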AnimBlueprint /BRCosmetics/Accessories/Hats/F_MED_Cacti_Hat/Meshes/F_MED_Cacti_FaceAcc_AnimBp
target_skeleton=&/F_MED_Cacti_FaceAcc_Skeleton
graphs [
 AnimGraph {nodes [
   n1 = RigidBody(ComponentPose=Local To Component(LocalPose=Copy Pose From Mesh() {bUseAttachedParent=true}), Alpha=Get Hat_Alpha() -> (Hat_Alpha:double)) {
    bClampLinearTranslationLimitToRefPose=true
    bOverrideWorldGravity=true
    overridePhysicsAsset=&/F_MED_Cacti_FaceAcc_Physics
    alphaBoolBlend {blendOption=Linear}
    componentLinearAccScale=<1 1 1>
    componentLinearVelScale=<1 1 1>
    overrideWorldGravity {z=-500}
   }
   Output Pose(Result=Component To Local(ComponentPose=@n1)) {layerGroup=None}
  ]}
 EventGraph {nodes [
   Event BlueprintUpdateAnimation()
   > Set Hat_Alpha(Hat_Alpha:double=SelectFloat(A=0.25, B=1, bPickA=Get bIsTargeting())) -> (Output_Get:double)
  ]}
]
variables [Hat_Alpha {type=number}]
